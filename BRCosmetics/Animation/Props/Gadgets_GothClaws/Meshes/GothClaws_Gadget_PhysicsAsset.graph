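# Physics asset for the GothClaws gadget prop mesh in BRCosmetics, giving its speaker bones simple collision shapes and joints.
# Three capsule bodies cover the central JNT_Speaker bone and its left and right speaker bones.
# Two constraints tie each side speaker to the central one, with equal angular limit values on both swing axes and twist.
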PhysicsAsset /BRCosmetics/Animation/Props/Gadgets_GothClaws/Meshes/GothClaws_Gadget_PhysicsAsset
bodies {
 JNT_Speaker [{shapeType=Capsule, length=58.6906, radius=24.7271, center=<0 0.9093 16.6669>, rotation=<rot 0 -90.0003 -89.9999>}]
 JNT_Speaker_l [{shapeType=Capsule, length=20.8181, radius=20.7789, center=<0 16.1667 0>, rotation=<rot 0 -10.9938 -76.3713>}]
 JNT_Speaker_r [{shapeType=Capsule, length=21.7747, radius=21.4418, center=<0 16.1667 0>, rotation=<rot 0 -165.6396 -72.4985>}]
}
constraints [
 {bone1Name=JNT_Speaker_l, bone2Name=JNT_Speaker, swing1LimitDegrees=45, swing2LimitDegrees=45, twistLimitDegrees=45}
 {bone1Name=JNT_Speaker_r, bone2Name=JNT_Speaker, swing1LimitDegrees=45, swing2LimitDegrees=45, twistLimitDegrees=45}
]
